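# Niagara ribbon trail system for the DoubleDuty pickaxe cosmetic, with a main layer and a wood-toned layer.
# Both emitters use FN AnimTrail modules to spawn ribbons between two skeletal mesh bone or socket indices, switched by User.IsR2.
# User.IsFourth lengthens particle lifetime, while curves shrink ribbon width and fade alpha.
# Time-of-day inputs adjust brightness on both layers and tint on the wood layer.
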NiagaraSystem /BRCosmetics/Weapons/FORT_Melee/Pickaxe_DoubleDuty/FX/NS_Pickaxe_DoubleDuty_Trail
dependencies {
 usedDataInterfaces [&NiagaraDataInterfaceCurve, &NiagaraDataInterfaceColorCurve, &NiagaraDataInterfaceSkeletalMesh]
 usedDynamicInputs []
 usedModules [
  &self:SystemSpawnScript.SystemScriptSource.SystemScriptGraph.NiagaraNodeAssignment_0.SetVariables_#1
  &SystemState
 ]
}
emitters {
 Trail_Main001 {
  emitterUpdateScript [
   &EmitterState(Life Cycle Mode=::NewEnumerator1, Loop Behavior=::NewEnumerator1, Loop Delay=0.2, Loop Duration=0.27, UseLoopDelay=true)
   &FN_SpawnAnimTrail_v2()
  ]
  particleSpawnScript [
   /Game/Effects/Niagara/Modules/AnimTrail/FN_InitializeAnimTrail/FN_InitializeAnimTrailBoneIndecies(
    First Bone or Socket Index:NiagaraInt32=&MakeCustomIntFromBool(Bool=$User.IsR2, True INT=2, False INT=0)
    Second Bone or Socket Index:NiagaraInt32=&MakeCustomIntFromBool(Bool=$User.IsR2, True INT=3, False INT=1)
   )
   &FN_InitializeAnimTrail(
    Color=<rgba 1 1 1 0.3>
    Lifetime=&MakeCustomFloatFromBool(Bool=$User.IsFourth, True Float=0.25, False Float=0.17)
    SkeletalMesh=$System.Skeletal Mesh
   ) {module=FN_InitializeAnimTrail001}
  ]
  particleUpdateScript [
   &FN_UpdateAnimTrail(SkeletalMesh=$System.Skeletal Mesh)
   &ScaleRibbonWidth(Ribbon Width Scale=&FloatFromCurve(FloatCurve=DataInterface() {propertyValues {Curve {keys [
         {arriveTangent=-1.2579, interpMode=RCIM_Cubic, time=0, value=1}
         {arriveTangent=-2.1146, interpMode=RCIM_Cubic, leaveTangent=-2.1146, tangentMode=RCTM_User, time=1, value=0}
        ]}}}))
   &ScaleColor(
    Scale Alpha=&FloatFromCurve(FloatCurve=DataInterface() {propertyValues {Curve {keys [
         {interpMode=RCIM_Cubic, leaveTangent=-2.3646, tangentMode=RCTM_Break, time=0, value=1}
         <at 0.6689 0.0233>
         {arriveTangent=-0.2139, interpMode=RCIM_Cubic, tangentMode=RCTM_Break, time=1, value=0}
        ]}}})
    Scale RGB=&VectorFromFloat(Value=&FNTimeOfDayScalar(Night Scale=0.7, Morning Scale=0.3, Evening Scale=0.3))
    ScaleA=true
    ScaleRGB=true
   )
  ]
  rendererClasses [&NiagaraRibbonRendererProperties]
  renderers [{rendererClass=&NiagaraRibbonRendererProperties}]
 }
 Trail_Wood001 {
  emitterUpdateScript [
   &EmitterState(Life Cycle Mode=::NewEnumerator1, Loop Behavior=::NewEnumerator1, Loop Delay=0.2, Loop Duration=0.27, UseLoopDelay=true)
   &FN_SpawnAnimTrail_v2()
  ]
  particleSpawnScript [
   &FN_InitializeAnimTrailBoneIndecies(
    First Bone or Socket Index:NiagaraInt32=&MakeCustomIntFromBool(Bool=$User.IsR2, True INT=2, False INT=0)
    Second Bone or Socket Index:NiagaraInt32=&MakeCustomIntFromBool(Bool=$User.IsR2, True INT=3, False INT=1)
   )
   &FN_InitializeAnimTrail(
    Color=&FNTimeOfDayColor(Night Color=<rgba 1 1 1 4>, Morning Color=<rgba 1 0.4909 0.1422 4>, Day Color=<rgba 1 1 1 4>, Evening Color=<rgba 1 0.4909 0.1422 4>)
    Lifetime=&MakeCustomFloatFromBool(Bool=$User.IsFourth, True Float=0.13, False Float=0.1)
    SkeletalMesh=$System.Skeletal Mesh
   ) {module=FN_InitializeAnimTrail001}
  ]
  particleUpdateScript [
   &FN_UpdateAnimTrail(SkeletalMesh=$System.Skeletal Mesh)
   &ScaleRibbonWidth(Ribbon Width Scale=&FloatFromCurve(FloatCurve=DataInterface() {propertyValues {Curve {keys [
         {arriveTangent=-1.2579, interpMode=RCIM_Cubic, time=0, value=1}
         {arriveTangent=-2.1146, interpMode=RCIM_Cubic, leaveTangent=-2.1146, tangentMode=RCTM_User, time=1, value=0}
        ]}}}))
   &ScaleColor(
    Scale Alpha=&FloatFromCurve(FloatCurve=DataInterface() {propertyValues {Curve {keys [
         {interpMode=RCIM_Cubic, leaveTangent=-2.3646, tangentMode=RCTM_Break, time=0, value=1}
         <at 0.6689 0.0233>
         {arriveTangent=-0.2139, interpMode=RCIM_Cubic, tangentMode=RCTM_Break, time=1, value=0}
        ]}}})
    Scale RGB=&MakeVectorFromLinearColorRGB(Color=&ColorFromCurve(ColorCurve=DataInterface() {propertyValues {
        AlphaCurve {keys [<at 0 1 RCIM_Cubic>,<at 1 1 RCIM_Cubic>]}
        BlueCurve {keys [<at 0.0029 0.0207 RCIM_Cubic>,<at 0.1525 0.0461 RCIM_Cubic>]}
        GreenCurve {keys [<at 0.0029 0.0924 RCIM_Cubic>,<at 0.1525 0.0653 RCIM_Cubic>]}
        RedCurve {keys [<at 0.0029 0.151 RCIM_Cubic>,<at 0.1525 0.11 RCIM_Cubic>]}
       }}))
    ScaleA=true
    ScaleRGB=true
   )
   &ScaleColor(
    Scale Alpha=&FNTimeOfDayScalar(Night Scale=0.75, Morning Scale=1, Evening Scale=1)
    Scale RGB=&VectorFromFloat(Value=&FNTimeOfDayScalar(Night Scale=0.4, Morning Scale=1, Day Scale=0.8, Evening Scale=0.8))
    ScaleA=true
    ScaleRGB=true
   ) {module=ScaleColor001}
  ]
  rendererClasses [&NiagaraRibbonRendererProperties]
  renderers [{rendererClass=&NiagaraRibbonRendererProperties}]
 }
}
summary {
 systemName=NS_Pickaxe_DoubleDuty_Trail
 emitters [
  {emitterName=Trail_Main001, rendererClasses [&NiagaraRibbonRendererProperties]}
  {emitterName=Trail_Wood001, rendererClasses [&NiagaraRibbonRendererProperties]}
 ]
 userVariables [
  User.IsFourth {type=NiagaraBool}
  User.IsR2 {type=NiagaraBool}
  User.RibbonIDOffset {type=NiagaraInt32}
 ]
}
